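# Sound cue for the landing sound of a dropped SunRose modular DMR weapon.
# An equally weighted random node picks one of three landing waves, feeding through a short delay into a modulator at default settings.
# It uses the shared weapon-drop attenuation asset and loot-collect sound class.
# A concurrency override caps simultaneous instances at a few and blocks new ones beyond that.
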SoundCue /SunRoseWeaponsGameplay/Items/Weapons/ModularDMR/Sounds/SunRose_DMR_Drop_Land_Cue
root=Modulator_0
nodes [
 {id=Delay_0, children [Random_0], properties {delayMin=0.1}}
 {id=Modulator_0, children [Delay_0]}
 {id=Random_0, children [WavePlayer_9, WavePlayer_10, WavePlayer_11], properties {weights [1,1,1]}}
 {id=WavePlayer_10, properties {soundWaveAssetPtr=&/SunRose_DMR_Drop_Land_02}}
 {id=WavePlayer_11, properties {soundWaveAssetPtr=&/SunRose_DMR_Drop_Land_01}}
 {id=WavePlayer_9, properties {soundWaveAssetPtr=&/SunRose_DMR_Drop_Land_03}}
]
properties {
 attenuationSettings=&Fort_Weapon_Drop_Att
 bOverrideConcurrency=true
 maxDistance=3000
 soundClassObject=&Fort_Loot_Collect
 volumeMultiplier=0.8
 attenuationOverrides {falloffDistance=1000, attenuationShapeExtents {x=0}}
 concurrencyOverrides {maxCount=3, resolutionRule=PreventNew}
}
