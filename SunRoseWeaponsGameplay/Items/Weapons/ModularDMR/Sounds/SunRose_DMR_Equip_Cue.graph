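# Sound cue for the equip foley of the SunRose modular DMR weapon.
# A random node picks one of three equip recordings with equal weight, then a modulator applies slight pitch and volume variation.
# It uses the weapon foley sound class and reload attenuation, overrides concurrency, and has the debug flag left on.
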SoundCue /SunRoseWeaponsGameplay/Items/Weapons/ModularDMR/Sounds/SunRose_DMR_Equip_Cue
root=Modulator_0
nodes [
 {id=Modulator_0, children [Random_1], properties {pitchMax=1.01, pitchMin=0.98, volumeMin=0.98}}
 {id=Random_1, children [WavePlayer_3,WavePlayer_4,WavePlayer_5], properties {weights [1,1,1]}}
 {id=WavePlayer_3, properties {soundWaveAssetPtr=&/SunRose_DMR_Equip_01}}
 {id=WavePlayer_4, properties {soundWaveAssetPtr=&/SunRose_DMR_Equip_02}}
 {id=WavePlayer_5, properties {soundWaveAssetPtr=&/SunRose_DMR_Equip_03}}
]
properties {
 attenuationSettings=&Fort_Weapon_Reload_Att
 bDebug=true
 bOverrideConcurrency=true
 maxDistance=2250
 soundClassObject=&Fort_WeaponFoley
 volumeMultiplier=0.501
}
