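# Sound cue for the magazine insert foley during a Sun Rose modular DMR reload.
# A Random node picks between two equally weighted wave players, the Insert_02 and Insert_03 variations.
# The cue routes to the weapon foley sound class with shared reload attenuation and a lowered volume, and it overrides concurrency and has its debug flag enabled.
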SoundCue /SunRoseWeaponsGameplay/Items/Weapons/ModularDMR/Sounds/SunRose_DMR_Reload_Insert_Cue
root=Random_1
nodes [
 {id=Random_1, children [WavePlayer_1,WavePlayer_2], properties {weights [1,1]}}
 {id=WavePlayer_1, properties {soundWaveAssetPtr=&/SunRose_DMR_Reload_Insert_02}}
 {id=WavePlayer_2, properties {soundWaveAssetPtr=&/SunRose_DMR_Reload_Insert_03}}
]
properties {
 attenuationSettings=&Fort_Weapon_Reload_Att
 bDebug=true
 bOverrideConcurrency=true
 maxDistance=2250
 soundClassObject=&Fort_WeaponFoley
 volumeMultiplier=0.319
}
