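# Sound cue for the reload-start foley of the SunRose modular DMR weapon.
# A Random node picks one of three equally weighted wave variations and feeds it into an Enveloper that fades volume to silence in under a second.
# The cue uses the weapon foley sound class and shared reload attenuation at reduced volume, with concurrency override and debug enabled.
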SoundCue /SunRoseWeaponsGameplay/Items/Weapons/ModularDMR/Sounds/SunRose_DMR_Reload_Start_Cue
root=Enveloper_0
nodes [
 {
  id=Enveloper_0
  children [Random_1]
  properties {pitchCurve {editorCurveData {keys [{time=0}]}}, volumeCurve {editorCurveData {keys [<at 0 1>,<at 0.6 0>]}}}
 }
 {id=Random_1, children [WavePlayer_3,WavePlayer_4,WavePlayer_5], properties {weights [1,1,1]}}
 {id=WavePlayer_3, properties {soundWaveAssetPtr=&/SunRose_DMR_Reload_Start_01}}
 {id=WavePlayer_4, properties {soundWaveAssetPtr=&/SunRose_DMR_Reload_Start_02}}
 {id=WavePlayer_5, properties {soundWaveAssetPtr=&/SunRose_DMR_Reload_Start_03}}
]
properties {
 attenuationSettings=&Fort_Weapon_Reload_Att
 bDebug=true
 bOverrideConcurrency=true
 maxDistance=2250
 soundClassObject=&Fort_WeaponFoley
 volumeMultiplier=0.451
}
